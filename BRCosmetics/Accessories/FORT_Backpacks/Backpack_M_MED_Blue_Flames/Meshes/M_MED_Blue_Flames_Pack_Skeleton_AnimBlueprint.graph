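# Animation Blueprint for the Blue Flames backpack cosmetic that spins its three turbines and shifts the pack with the wearer's spine bend.
# Each update derives turbine roll from real time, switching to a faster rate when pawn 2D speed exceeds a threshold.
# The owner mesh's spine_05-to-spine_01 delta yaw maps to clamped root roll and Y offset.
# AnimGraph chains Modify Bone nodes for root and turbines; two turbine translation nodes remain unconnected to output.
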AnimBlueprint /BRCosmetics/Accessories/FORT_Backpacks/Backpack_M_MED_Blue_Flames/Meshes/M_MED_Blue_Flames_Pack_Skeleton_AnimBlueprint
target_skeleton=&/M_MED_Blue_Flames_Pack_Skeleton
graphs [
 AnimGraph {nodes [
   n1 = "Transform (Modify) Bone - Bone: dyn_turbine_r"(Alpha=1, Translation="0.5,0,0", Rotation="0,0,0", Scale="1,1,1") {
    rotationMode=BMM_Additive
    rotationSpace=BCS_BoneSpace
    translationMode=BMM_Additive
    translationSpace=BCS_BoneSpace
    alphaBoolBlend {blendOption=Linear}
    boneToModify {boneName=dyn_turbine_r}
    translation {x=0.5}
   }
   "Transform (Modify) Bone - Bone: dyn_turbine_l"(ComponentPose=@n1, Alpha=1, Translation="0.5,0,0", Rotation="0,0,0", Scale="1,1,1") {
    rotationMode=BMM_Additive
    rotationSpace=BCS_BoneSpace
    translationMode=BMM_Additive
    translationSpace=BCS_BoneSpace
    alphaBoolBlend {blendOption=Linear}
    boneToModify {boneName=dyn_turbine_l}
    translation {x=0.5}
   }
   n2 = "Transform (Modify) Bone - Bone: root"(ComponentPose=Local To Component(LocalPose=Input Pose()), Translation=Get backpack_translation(), Rotation=Get backpack_rotate()) {
    rotationMode=BMM_Additive
    rotationSpace=BCS_BoneSpace
    translationMode=BMM_Additive
    translationSpace=BCS_BoneSpace
    alphaBoolBlend {blendOption=Linear}
    boneToModify {boneName=root}
   }
   n3 = Get spinnerRot() -> (spinnerRot:Rotator)
   n4 = "Transform (Modify) Bone - Bone: dyn_turbine_top"(ComponentPose=@n2, Alpha=1, Translation="0,0,0", Rotation=@n3, Scale="1,1,1") {rotationMode=BMM_Replace, rotationSpace=BCS_BoneSpace, alphaBoolBlend {blendOption=Linear}, boneToModify {boneName=dyn_turbine_top}}
   n5 = "Transform (Modify) Bone - Bone: dyn_turbine_r"(ComponentPose=@n4, Alpha=1, Translation="0,0,0", Rotation=@n3, Scale="1,1,1") {rotationMode=BMM_Replace, rotationSpace=BCS_BoneSpace, alphaBoolBlend {blendOption=Linear}, boneToModify {boneName=dyn_turbine_r}}
   n6 = "Transform (Modify) Bone - Bone: dyn_turbine_l"(ComponentPose=@n5, Alpha=1, Translation="0,0,0", Rotation=@n3, Scale="1,1,1") {rotationMode=BMM_Replace, rotationSpace=BCS_BoneSpace, alphaBoolBlend {blendOption=Linear}, boneToModify {boneName=dyn_turbine_l}}
   Output Pose(Result=Component To Local(ComponentPose=@n6)) {layerGroup=None}
  ]}
 EventGraph {nodes [
   Event BlueprintUpdateAnimation()
   > n1 = Sequence()
   n2 = float / float(A=GetRealTimeSeconds(), B=720) {is=title K2Node_CallFunction}
   > Set spinnerRot(spinnerRot:Rotator=MakeRotator(Roll=float * float(A=@n2, B=100000))) -> (Output_Get:Rotator)
   > Branch(Condition="float > float"(A=Get PawnSpeed2D(), B=50) {is=title K2Node_CallFunction})
   n3 = GetRealTimeSeconds()
   > Set spinnerRot(spinnerRot=MakeRotator(Roll=float * float(A=@n3, B=500)))
   n4 = GetOwnerSkeletalMeshComponent()
   n5 = BreakRotator(InRot=BreakTransform(InTransform=GetDeltaTransformFromRefPose(self=@n4, BoneName=spine_05, BaseName=spine_01)).Rotation)
   Set backpack_rotate(execute=@n1.then_1, backpack_rotate=MakeRotator(Roll=MapRangeClamped(Value=@n5.Yaw, InRangeA=0, InRangeB=-30, OutRangeB=20)))
   > Set backpack_translation(backpack_translation=MakeVector(Y=MapRangeClamped(Value=@n5.Yaw, InRangeA=0, InRangeB=-30, OutRangeB=-3)))
  ]}
]
variables [
 backpack_rotate {type=Rotator}
 backpack_translation {type=Vector}
 spinnerRot {type=Rotator}
]
